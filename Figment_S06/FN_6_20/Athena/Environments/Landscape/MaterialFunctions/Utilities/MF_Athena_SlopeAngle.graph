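# Material function for Athena landscape materials that derives a slope value from the world-space vertex normal.
# It takes the normal's Z component and multiplies it by the AngleBlend_Offset parameter from the Rock group, feeding that product straight to the output without clamping.
# A lerp between Angle_Min and Angle_Max plus the offset, and a Clamp on the product, are present but do not reach the output.
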MaterialFunction /Figment_S06/FN_6_20/Athena/Environments/Landscape/MaterialFunctions/Utilities/MF_Athena_SlopeAngle
graph {nodes [
  n1 = ComponentMask(None=VertexNormalWS()) {b=true}
  n2 = ScalarParameter() {defaultValue=2, group=Rock, parameterName=AngleBlend_Offset}
  Add(
   A=LinearInterpolate(A=ScalarParameter() {parameterName=Angle_Min}, B=ScalarParameter() {defaultValue=0.5, parameterName=Angle_Max}, Alpha=@n1)
   B=@n2
  )
  n3 = Multiply(A=@n1, B=@n2)
  Clamp(None=@n3)
  FunctionOutput(None=@n3)
 ]}
